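# Material function that sets up Pixel Depth Offset for foliage, giving meshes per-pixel depth from a depth texture.
# CameraDepthFade, driven by dynamic shadow distance, scales a FoliageScaleFactor radius, output through a CustomUV to reach the pixel shader.
# World Position Offset pulls vertices toward the camera by that radius times a PixelOffset parameter.
# Pixel Depth Offset pushes pixels back by the corresponding diameter times the linear depth texture.
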MaterialFunction /Engine/Functions/Engine_MaterialFunctions01/Lighting/PixelDepthOffset_Foliage
graph {
 properties {
  bExposeToLibrary=true
  description=This function helps set up PixelDepthOffset for foliage meshes.
  libraryCategoriesText [Misc,Foliage,Lighting]
 }
 nodes [
  n1 = FunctionInput() {bUsePreviewValueAsDefault=true, description=WorldPositionOffset before PixelDepth modifications, inputName=WorldPositionOffset}
  n2 = WorldPosition() {worldPositionShaderOffset=WPT_ExcludeAllShaderOffsets}
  n3 = Normalize(VectorInput=Subtract(A=Add(A=@n1, B=@n2.XYZ), B=CameraPositionWS())) {desc=CameraVector}
  n4 = FunctionInput() {
   bUsePreviewValueAsDefault=true
   description=This will mask the PixelDepthOffset by the dynamic shadow distance. This is necessary when using for lighting since PDO only works with Distance Field shadows currently. If 0 it is disabled.
   inputName=Dynamic Shadow Distance
   inputType=FunctionInput_Scalar
   sortPriority=1
  }
  n5 = FunctionInput(Preview=Multiply(A=@n4) {constB=0.1}) {
   bUsePreviewValueAsDefault=true
   description=This is how long the fade length for the dynamic shadow mask is. Defaults to 10% of Shadow Distance input.
   inputName=Dynamic Shadow Fade Length
   inputType=FunctionInput_Scalar
   sortPriority=2
  }
  n6 = StaticBool() {value=true}
  n7 = FunctionInput() {
   bUsePreviewValueAsDefault=true
   description="This is the radius of the mesh this will be used on, without any scaling. This is required since foliage does not have access to the objectradius or bounds."
   inputName=Initial Radius
   inputType=FunctionInput_Scalar
   sortPriority=3
   previewValue=<500 0 0 1>
  }
  n8 = FunctionOutput(None=Multiply(
    A=MaterialFunctionCall(Fade Length=@n5, Fade Offset=@n4, For Vertex Shader=@n6) {materialFunction=&CameraDepthFade}.Result
    B=MaterialFunctionCall!(Initial Radius=@n7) {materialFunction=/Engine/Functions/Engine_MaterialFunctions01/WorldPositionOffset/FoliageScaleFactor}.Result
   )) {
   description=This output needs to be hooked up to one of the material's unused CustomUV input pins. Then the matching TextureCoordinate node needs to be input on the INPUT side of this function node.
   outputName=Custom UVs
   sortPriority=1
  }
  n9 = ScalarParameter() {defaultValue=1, parameterName=PixelOffset}
  n10 = Multiply(A=@n8, B=@n9)
  FunctionOutput(None=Subtract(A=@n1, B=Multiply(A=@n3, B=@n10))) {
   description=Pulle the object fowards the camera by object Radius. Connect this to WorldPositionOffset in the material.
   outputName=World Position Offset
  }
  n11 = FunctionInput(Preview=TextureCoordinate() {coordinateIndex=2}) {
   bUsePreviewValueAsDefault=true
   description=This needs to be the same UV channel as the one connected to the CustomUV output of this function.
   inputName=CustomUV Input
   inputType=FunctionInput_Vector2
   sortPriority=4
   previewValue=<500 0 0 1>
  }
  n12 = Multiply(A=@n9, B=ComponentMask(None=@n11) {r=true})
  n13 = FunctionInput() {
   bUsePreviewValueAsDefault=true
   description=This should be a linear depth texture. It should be in the 0-1 range with values closest to the camera being black and values further from the camera being white. This is inverse of typical depth/heightmaps but makes this function cheaper.
   inputName=Depth Texture
   inputType=FunctionInput_Scalar
   sortPriority=5
  }
  FunctionOutput(None=Multiply(A=Multiply(A=@n12) {constB=2, desc=Radius to Diameter}, B=@n13) {constB=2, desc=Radius to Diameter}) {outputName=Pixel Depth Offset, sortPriority=2}
 ]
}
