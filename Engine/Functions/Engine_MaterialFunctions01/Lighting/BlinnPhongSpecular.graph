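# BlinnPhongSpecular is a library material function in the Lighting category that returns a Blinn-Phong specular lobe.
# It normalizes Light Vector plus ViewVector into a half vector, dots it with Normal, clamps at zero and raises it to Glossiness, used directly as exponent.
# ViewVector defaults to a transformed view-space Z axis.
# The description claims GGX, but the graph contains only this power-based lobe.
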MaterialFunction /Engine/Functions/Engine_MaterialFunctions01/Lighting/BlinnPhongSpecular
graph {
 properties {bExposeToLibrary=true, description=Returns a specular lobe using GGX., libraryCategoriesText [Lighting]}
 nodes [
  n1 = FunctionInput() {bUsePreviewValueAsDefault=true, description=The surface normal., inputName=Normal, previewValue=<0 0 1 1>}
  n2 = FunctionInput() {
   bUsePreviewValueAsDefault=true
   description=The light vector to calculate specular for.
   inputName=Light Vector
   sortPriority=1
   previewValue=<0 0 1 1>
  }
  n3 = FunctionInput(Preview=Transform(None=Constant3Vector() {constant=<rgba 0 0 1 0>}) {transformSourceType=TRANSFORMSOURCE_View}) {bUsePreviewValueAsDefault=true, description=The view direction, inputName=ViewVector, sortPriority=3, previewValue=<0 0 1 1>}
  n4 = DotProduct(A=@n1, B=Normalize(VectorInput=Add(A=@n2, B=@n3)))
  n5 = FunctionInput() {
   bUsePreviewValueAsDefault=true
   description=The glossiness of the surface
   inputName=Glossiness
   inputType=FunctionInput_Scalar
   sortPriority=4
   previewValue=<0.7 0 0 1>
  }
  FunctionOutput(None=Power(Base=Max(A=@n4) {constB=0}, Exp=@n5))
 ]
}
