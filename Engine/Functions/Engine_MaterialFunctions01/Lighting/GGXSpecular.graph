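# GGXSpecular is a library material function in the Lighting category that returns a scalar GGX specular lobe for a normal, light vector and roughness.
# A single Custom HLSL node evaluates only the GGX distribution term, using roughness to the fourth power as alpha squared, with no Fresnel or visibility terms.
# It takes no view vector, so the half vector is the normalized sum of normal and light vector.
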MaterialFunction /Engine/Functions/Engine_MaterialFunctions01/Lighting/GGXSpecular
graph {
 properties {bExposeToLibrary=true, description=Returns a specular lobe using GGX., libraryCategoriesText [Lighting]}
 nodes [
  MaterialExpressionFunctionInput_0 = FunctionInput() {bUsePreviewValueAsDefault=true, description=The surface normal., inputName=Normal, previewValue=<0 0 1 1>}
  MaterialExpressionFunctionInput_1 = FunctionInput() {
   bUsePreviewValueAsDefault=true
   description=The light vector to calculate specular for.
   inputName=Light Vector
   sortPriority=1
   previewValue=<0 0 1 1>
  }
  MaterialExpressionFunctionInput_2 = FunctionInput() {
   bUsePreviewValueAsDefault=true
   description=The surface roughness
   inputName=Roughness
   inputType=FunctionInput_Scalar
   sortPriority=3
   previewValue=<0.1 0 0 1>
  }
  n1 = Custom!(Normal=@MaterialExpressionFunctionInput_0, LightVector=@MaterialExpressionFunctionInput_1, Roughness=@MaterialExpressionFunctionInput_2) {
   code="\t\t\tfloat NoH = dot( normalize(Normal + LightVector), Normal);\r\n\t\r\n\t\t\tfloat a2 = Pow4( Roughness );\r\n\t\t\tfloat d = ( NoH * a2 - NoH ) * NoH + 1;\t// 2 mad\r\n\t\t\treturn a2 / ( PI*d*d );\t\t\t\t\t// 4 mul, 1 rcp"
   outputType=CMOT_Float1
   inputs [
    {inputName=Normal, input {expression=&self:MaterialExpressionFunctionInput_0}}
    {
     inputName=LightVector
     input {expression=&self:MaterialExpressionFunctionInput_1}
    }
    {inputName=Roughness, input {expression=&self:MaterialExpressionFunctionInput_2}}
   ]
  }
  FunctionOutput(None=@n1)
 ]
}
